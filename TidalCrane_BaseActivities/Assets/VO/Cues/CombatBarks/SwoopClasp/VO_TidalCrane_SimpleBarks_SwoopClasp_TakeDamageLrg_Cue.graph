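# Sound cue for SwoopClasp combat bark voice lines in TidalCrane base activities, named for taking large damage.
# A root Random node picks among eight equally weighted wave players, each holding a different recorded take.
# It plays through the Fort_VO_Narrator sound class with shared base-activities attenuation and a SwoopClasp-specific concurrency asset that governs simultaneous barks.
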SoundCue /TidalCrane_BaseActivities/Assets/VO/Cues/CombatBarks/SwoopClasp/VO_TidalCrane_SimpleBarks_SwoopClasp_TakeDamageLrg_Cue
root=Random_0
nodes [
 {
  id=Random_0
  children [WavePlayer_0,WavePlayer_1,WavePlayer_2,WavePlayer_3,WavePlayer_4,WavePlayer_5,WavePlayer_6,WavePlayer_7]
  properties {weights [1,1,1,1,1,1,1,1]}
 }
 {
  id=WavePlayer_0
  properties {soundWaveAssetPtr=/TidalCrane_BaseActivities/Assets/VO/Waves/SwoopClasp/VO_TidalCrane_SimpleBarks_SwoopClasp_TakeDamageLrg_010}
 }
 {
  id=WavePlayer_1
  properties {soundWaveAssetPtr=/TidalCrane_BaseActivities/Assets/VO/Waves/SwoopClasp/VO_TidalCrane_SimpleBarks_SwoopClasp_TakeDamageLrg_011}
 }
 {
  id=WavePlayer_2
  properties {soundWaveAssetPtr=/TidalCrane_BaseActivities/Assets/VO/Waves/SwoopClasp/VO_TidalCrane_SimpleBarks_SwoopClasp_TakeDamageLrg_012}
 }
 {
  id=WavePlayer_3
  properties {soundWaveAssetPtr=/TidalCrane_BaseActivities/Assets/VO/Waves/SwoopClasp/VO_TidalCrane_SimpleBarks_SwoopClasp_TakeDamageLrg_013}
 }
 {
  id=WavePlayer_4
  properties {soundWaveAssetPtr=/TidalCrane_BaseActivities/Assets/VO/Waves/SwoopClasp/VO_TidalCrane_SimpleBarks_SwoopClasp_TakeDamageLrg_014}
 }
 {
  id=WavePlayer_5
  properties {soundWaveAssetPtr=/TidalCrane_BaseActivities/Assets/VO/Waves/SwoopClasp/VO_TidalCrane_SimpleBarks_SwoopClasp_TakeDamageLrg_015}
 }
 {
  id=WavePlayer_6
  properties {soundWaveAssetPtr=/TidalCrane_BaseActivities/Assets/VO/Waves/SwoopClasp/VO_TidalCrane_SimpleBarks_SwoopClasp_TakeDamageLrg_016}
 }
 {
  id=WavePlayer_7
  properties {soundWaveAssetPtr=/TidalCrane_BaseActivities/Assets/VO/Waves/SwoopClasp/VO_TidalCrane_SimpleBarks_SwoopClasp_TakeDamageLrg_017}
 }
]
properties {
 attenuationSettings=/TidalCrane_BaseActivities/Assets/VO/AS_Fort_BaseActivities
 maxDistance=10100
 soundClassObject=/Game/Sounds/FortSoundClassesAndModes/Fort_VO_Narrator
 volumeMultiplier=1
 concurrencySet [/TidalCrane_BaseActivities/Assets/VO/BaseActivites_VO_SwoopClasp_Conc]
}
